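; PCG biome transform graph that scatters secondary points around each input point and projects them onto the landscape.
; A grid stores normalized center distance as density, is trimmed to a rough disc by a density filter, then copied onto every input point.
; Half the copies are kept and given random offset, yaw and scale.
; Scale is multiplied by density remapped between ScaleNear and ScaleFar, using the measured density extremes as input range.
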(graph BasicSecondaries
  (params (DistanceFromSourcePoint :type Double :default 1e+03) (PointSpacing :type Double :default 2.5e+02) (Seed :type Int :default -980430391) (ScaleMin :type Double :default 0.75) (ScaleMax :type Double :default 1.25) (ScaleNear :type Double :default 1.0) (ScaleFar :type Double :default 1.0))
  (bind CreateAttribute_22 (Create%20Constant :AttributeTypes (struct :DoubleValue 1.0)))
  (bind AttributeMathsOp_2 (Attribute%20Maths%20Op :InputSource1 $Scale :Operation Set))
  (bind CreateAttribute_17 (Create%20Constant :AttributeTypes (struct :DoubleValue 2.0)))
  (bind AttributeMathsOp_16 (Attribute%20Maths%20Op :Operation Sqrt))
  (bind AttributeMathsOp_15 (Attribute%20Maths%20Op :Operation Multiply :InA DistanceFromSourcePoint))
  (bind CreateAttribute_9 (Create%20Constant :AttributeTypes (struct :Type Vector)))
  (bind MakeVectorAttribute_19 (Make%20Vector%20Attribute :OutputType Vector :X DistanceFromSourcePoint :Y DistanceFromSourcePoint))
  (bind CreatePointsGrid_8 (Create%20Points%20Grid :bSetPointsBounds false :CellSize (2.5e+02 2.5e+02 2.5e+02) :GridExtents (2e+03 2e+03 0.0) :PointSteepness 0.5 :CellSize-param PointSpacing))
  (bind AttributeVectorOp_8 (Attribute%20Vector%20Op :InputSource1 $Position :Operation Distance :OutputTarget $Density))
  (bind AttributeMathsOp_11 (Attribute%20Maths%20Op :Operation Divide))
  (bind MutateSeed_3 (Mutate%20Seed :Seed-param Seed))
  (bind DensityFilter_18 (Density%20Filter :LowerBound 0.0 :UpperBound 0.75))
  (bind AttributeMathsOp_21 (Attribute%20Maths%20Op :InputSource1 $Steepness :Operation Set))
  (bind CopyPoints_0 (Copy%20Points :AttributeInheritance TargetOnly :ScaleInheritance Source))
  (bind SelectPoints_0 (Select%20Points :Ratio 0.5))
  (bind TransformPoints_2 (Transform%20Points :OffsetMax (1.5e+02 1.5e+02 0.0) :OffsetMin (-1.5e+02 -1.5e+02 0.0) :RotationMax (struct :Yaw 3.6e+02) :ScaleMax (5.0 3.0 1e+01) :ScaleMin (0.1 0.1 1.0) :ScaleMax-param ScaleMax :ScaleMin-param ScaleMin))
  (bind AttributeReduce_0 (Attribute%20Reduce :InputSource $Density :Operation Min))
  (bind AttributeReduce_2 (Attribute%20Reduce :InputSource $Density :Operation Max))
  (bind DensityRemap_13 (Density%20Remap :bExcludeValuesOutsideInputRange true :InRangeMax 0.0 :InRangeMin 1.0 :OutRangeMin 0.1 :OutRangeMax-param ScaleNear :OutRangeMin-param ScaleFar))
  (bind AttributeMathsOp_9 (Attribute%20Maths%20Op :InputSource1 $Scale :InputSource2 $Density :Operation Multiply))
  (bind GetLandscapeData_5 (Get%20Landscape%20Data :bComponentsMustOverlapSelf false :SamplingProperties (struct :bGetLayerWeights false)))
  (bind Projection_4 (Projection :ProjectionParams (struct :AttributeMode IncludeAttributes :AttributeMergeOperation SourceValue)))
  (bind Debug_2 (Debug))
  (connect DefaultInputNode In AttributeMathsOp_2 InA)
  (connect CreateAttribute_22 Out AttributeMathsOp_2 InB)
  (connect DensityRemap_13 Out AttributeMathsOp_9 InA)
  (connect DensityRemap_13 Out AttributeMathsOp_9 InB)
  (connect AttributeVectorOp_8 Out AttributeMathsOp_11 InA)
  (connect AttributeMathsOp_15 Out AttributeMathsOp_11 InB)
  (connect AttributeMathsOp_16 Out AttributeMathsOp_15 InB)
  (connect CreateAttribute_17 Out AttributeMathsOp_16 In)
  (connect DensityFilter_18 Out AttributeMathsOp_21 InA)
  (connect CreateAttribute_22 Out AttributeMathsOp_21 InB)
  (connect TransformPoints_2 Out AttributeReduce_0 In)
  (connect TransformPoints_2 Out AttributeReduce_2 In)
  (connect CreatePointsGrid_8 Out AttributeVectorOp_8 InA)
  (connect CreateAttribute_9 Out AttributeVectorOp_8 InB)
  (connect AttributeMathsOp_21 Out CopyPoints_0 Source)
  (connect AttributeMathsOp_2 Out CopyPoints_0 Target)
  (connect MakeVectorAttribute_19 Out CreatePointsGrid_8 GridExtents)
  (connect Projection_4 Out Debug_2 In)
  (connect MutateSeed_3 Out DensityFilter_18 In)
  (connect TransformPoints_2 Out DensityRemap_13 In)
  (connect AttributeReduce_0 Out DensityRemap_13 InRangeMax)
  (connect AttributeReduce_2 Out DensityRemap_13 InRangeMin)
  (connect AttributeMathsOp_11 Out MutateSeed_3 In)
  (connect AttributeMathsOp_9 Out Projection_4 In)
  (connect GetLandscapeData_5 Out Projection_4 Projection%20Target)
  (connect CopyPoints_0 Out SelectPoints_0 In)
  (connect SelectPoints_0 Out TransformPoints_2 In)
  (output Out Projection_4))
