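; DuplicatePattern is a PCG Biome Core transform graph that turns each input point into a radial pinwheel pattern of points on the landscape.
; Transform Points sets absolute rotation, then a first Duplicate Point makes in-place copies at even yaw steps around the full circle.
; A second Duplicate Point grows each into a curving arm by stepping along local X with added yaw and shrinking scale.
; Projection then drops the result onto Get Landscape Data.
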(graph DuplicatePattern
  (bind TransformPoints_0 (Transform%20Points :bAbsoluteRotation true))
  (bind DuplicatePoint_1 (Duplicate%20Point :Direction (0.0 0.0 0.0) :Iterations 9 :PointTransform (struct :Rotation (0.0 -0.0 0.30901699437494745 0.9510565162951536))))
  (bind DuplicatePoint_0 (Duplicate%20Point :bDirectionAppliedInRelativeSpace true :Direction (1.0 0.0 0.0) :Iterations 10 :PointTransform (struct :Rotation (0.0 -0.0 0.08715574274765803 0.9961946980917454) :Scale3D (0.8 0.8 0.8))))
  (bind GetLandscapeData_5 (Get%20Landscape%20Data :bComponentsMustOverlapSelf false :SamplingProperties (struct :bGetLayerWeights false)))
  (bind Projection_4 (Projection :ProjectionParams (struct :AttributeMode IncludeAttributes :AttributeMergeOperation SourceValue)))
  (connect DuplicatePoint_1 Out DuplicatePoint_0 In)
  (connect TransformPoints_0 Out DuplicatePoint_1 In)
  (connect DuplicatePoint_0 Out Projection_4 In)
  (connect GetLandscapeData_5 Out Projection_4 Projection%20Target)
  (connect DefaultInputNode In TransformPoints_0 In)
  (output Out Projection_4))
