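; DuplicatePoints is a PCG Biome Core transform graph that duplicates incoming points using the bounds of their mesh.
; It reads the Mesh attribute from a single indexed point and builds a mesh-sized point from it.
; That point's BoundsMin and BoundsMax are copied onto all input points before Duplicate Point iterates five times.
; The output merges the duplicates with the converted input points, which keep their original bounds.
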(graph DuplicatePoints
  (bind ToPoint_1 (To%20Point))
  (bind GetAttributeFromPointIndex_2 (Get%20Attribute%20From%20Point%20Index :bForceOutputAttributeToBeInElementsDomain true :InputSource Mesh :OutputAttributeName None))
  (bind PointFromMesh_1 (Point%20From%20Mesh))
  (bind AttributeMathsOp_3 (Attribute%20Maths%20Op :InputSource1 $BoundsMin :InputSource2 $BoundsMin :Operation Set))
  (bind AttributeMathsOp_5 (Attribute%20Maths%20Op :InputSource1 $BoundsMax :InputSource2 $BoundsMax :Operation Set))
  (bind DuplicatePoint_0 (Duplicate%20Point :Iterations 5))
  (connect ToPoint_1 Out AttributeMathsOp_3 InA)
  (connect PointFromMesh_1 Out AttributeMathsOp_3 InB)
  (connect AttributeMathsOp_3 Out AttributeMathsOp_5 InA)
  (connect PointFromMesh_1 Out AttributeMathsOp_5 InB)
  (connect AttributeMathsOp_5 Out DuplicatePoint_0 In)
  (connect ToPoint_1 Out GetAttributeFromPointIndex_2 In)
  (connect GetAttributeFromPointIndex_2 Attribute PointFromMesh_1 Mesh)
  (connect DefaultInputNode In ToPoint_1 In)
  (output Out DuplicatePoint_0)
  (output Out ToPoint_1))
