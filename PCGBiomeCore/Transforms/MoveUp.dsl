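; PCG biome core transform graph MoveUp that raises incoming data by a fixed vertical distance.
; The input is converted to points with a To Point node, then passed to Transform Points.
; Transform Points uses equal minimum and maximum Z offsets of 500 units, so every point moves up uniformly with no random spread, and the shifted points become the output.
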(graph MoveUp
  (bind ToPoint_1 (To%20Point))
  (bind TransformPoints_0 (Transform%20Points :OffsetMax (0.0 0.0 5e+02) :OffsetMin (0.0 0.0 5e+02)))
  (connect DefaultInputNode In ToPoint_1 In)
  (connect ToPoint_1 Out TransformPoints_0 In)
  (output Out TransformPoints_0))
